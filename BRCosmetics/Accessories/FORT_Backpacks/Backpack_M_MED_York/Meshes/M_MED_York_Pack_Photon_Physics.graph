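# Physics asset for the York Photon backpack cosmetic mesh, defining bodies and joints for the pack's proton, hose and gun cable bones.
# A master box body links to the proton box through a tight constraint that mainly allows one swing axis.
# The hose and two gun cables each constrain to the proton body, the hose with looser limits.
# The hose uses two angled capsules, while each cable uses a tiny capsule.
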PhysicsAsset /BRCosmetics/Accessories/FORT_Backpacks/Backpack_M_MED_York/Meshes/M_MED_York_Pack_Photon_Physics
bodies {
 dyn_gun_cable1 [{shapeType=Capsule, length=0.505, radius=0.505, center=<0.5 0 0>, rotation=<rot 90 0 0>}]
 dyn_gun_cable2 [{shapeType=Capsule, length=0.505, radius=0.505, center=<0.5 0 0>, rotation=<rot 90 0 0>}]
 dyn_hose_to_gun_ [
  {shapeType=Capsule, length=20.782, radius=1.5, center=<1 0 0>, rotation=<rot 0 -81.5851 -81.013>}
  {shapeType=Capsule, length=5, radius=1.5, center=<23.5836 2.3555 -3.3944>, rotation=<rot -90 26.565 -74.3748>}
 ]
 dyn_master_proton [{shapeType=Box, extentX=25, extentY=2, extentZ=11.5, center=<6.9038 -1.8403 0>}]
 dyn_proton [{shapeType=Box, extentX=25, extentY=2, extentZ=5, center=<7.5494 0.175 0.0427>, rotation=<rot -8.4377 0 0>}]
}
constraints [
 {bone1Name=dyn_gun_cable1, bone2Name=dyn_proton, swing1LimitDegrees=10, swing2LimitDegrees=30, twistLimitDegrees=2}
 {bone1Name=dyn_hose_to_gun_, bone2Name=dyn_proton, swing1LimitDegrees=45, swing2LimitDegrees=45, twistLimitDegrees=45}
 {bone1Name=dyn_gun_cable2, bone2Name=dyn_proton, swing1LimitDegrees=10, swing2LimitDegrees=30, twistLimitDegrees=2}
 {bone1Name=dyn_proton, bone2Name=dyn_master_proton, swing1LimitDegrees=1, swing2LimitDegrees=20, twistLimitDegrees=1}
]
