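# Physics asset for the York medium male backpack cosmetic, setting up physical motion for its dangling cables and hoses.
# A capsule body on the pack bone acts as the hub, with small capsules on seven dyn_ bones for the hoses, side cables and motherboard cable.
# Each dyn_ bone is constrained directly to the pack rather than chained, with per-bone swing and twist limits ranging from nearly rigid to moderately loose.
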PhysicsAsset /BRCosmetics/Accessories/FORT_Backpacks/Backpack_M_MED_York/Meshes/M_MED_York_Pack_Physics
bodies {
 dyn_cable2 [{shapeType=Capsule, length=1, radius=0.5, center=<0.5 0 0>, rotation=<rot 90 0 0>}]
 dyn_motherboard_cable [{shapeType=Capsule, length=5.1361, radius=2.1225, center=<1 0 0>, rotation=<rot 0 -98.2459 -81.5051>}]
 dyn_pack_hose [{shapeType=Capsule, length=2.6942, radius=1.4264, center=<1 0 0>, rotation=<rot 0 -134.9992 -79.3341>}]
 dyn_pack_hose_smallcable [{shapeType=Capsule, length=1, radius=0.5, center=<0.5 0 0>, rotation=<rot 90 0 0>}]
 dyn_pack_hose_smallcable1 [{shapeType=Capsule, length=0.505, radius=0.505, center=<0.2 0 0>, rotation=<rot 0 -45.7846 -78.3231>}]
 dyn_side_cable1 [{shapeType=Capsule, length=1, radius=0.5, center=<0.5 0 0>, rotation=<rot 90 0 0>}]
 dyn_side_cable_2 [{shapeType=Capsule, length=1, radius=0.5, center=<0.5 0 0>, rotation=<rot 90 0 0>}]
 pack [{shapeType=Capsule, length=5, radius=5}]
}
constraints [
 {bone1Name=dyn_cable2, bone2Name=pack, swing1LimitDegrees=30, swing2LimitDegrees=10, twistLimitDegrees=30}
 {bone1Name=dyn_pack_hose_smallcable, bone2Name=pack, swing1LimitDegrees=45, swing2LimitDegrees=10, twistLimitDegrees=45}
 {bone1Name=dyn_side_cable1, bone2Name=pack, swing1LimitDegrees=1, swing2LimitDegrees=10, twistLimitDegrees=1}
 {bone1Name=dyn_side_cable_2, bone2Name=pack, swing1LimitDegrees=1, swing2LimitDegrees=10, twistLimitDegrees=1}
 {bone1Name=dyn_pack_hose, bone2Name=pack, swing1LimitDegrees=45, swing2LimitDegrees=25, twistLimitDegrees=45}
 {bone1Name=dyn_motherboard_cable, bone2Name=pack, swing1LimitDegrees=45, swing2LimitDegrees=35, twistLimitDegrees=1}
 {bone1Name=dyn_pack_hose_smallcable1, bone2Name=pack, swing1LimitDegrees=10, swing2LimitDegrees=1, twistLimitDegrees=1}
]
